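# Additive unlit fog volume material that draws a soft glowing sphere, visible only within a limited camera distance range and tied to the hologram parameter collection.
# Opacity comes from a ray-traced sphere falloff with depth fade and optional hemisphere mask, scaled by near/far distance fades and ScreensEnabled.
# When that fade is near zero, a vertex offset collapses the mesh to its pivot.
# Color follows ColorSwitch, Captured and Contested states, with time-of-day brightness.
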Material /Game/Environments/TechArt/FogVolumes/M_SimpleVolumeFogLimitedRange
graph {
 properties {
  bAutomaticallySetUsageInEditor=false
  blendMode=BLEND_Additive
  pixelDepthOffsetMode=PDOM_Legacy
  shadingModel=MSM_Unlit
  usedShadingModels=Unlit
 }
 nodes [
  n1 = CameraPositionWS()
  n2 = Subtract(A=WorldPosition().XYZ, B=@n1) {constB=-1}
  ComponentMask(None=TransformPosition(None=@n2) {transformSourceType=TRANSFORMPOSSOURCE_World}) {b=true}
  n3 = Distance(A=ObjectPositionWS(), B=CameraPositionWS())
  n4 = Multiply(
   A=ScalarParameter() {defaultValue=74.2047, parameterName=Radius}
   B=MaterialFunctionCall() {materialFunction=&ObjectScale}.Scale X
  )
  n5 = ScalarParameter() {parameterName=NearFade_Offset}
  n6 = Divide(A=Subtract(A=@n3, B=Add(A=@n4, B=@n5)), B=ScalarParameter() {defaultValue=1000, parameterName=NearFade_Distance})
  n7 = ScalarParameter() {defaultValue=8665.7285, parameterName=FarFade_Offset}
  n8 = Saturate(None=Divide(A=Subtract(A=@n3, B=@n7), B=ScalarParameter() {defaultValue=20000, parameterName=FarFade_Distance}))
  n9 = Multiply(
   A=Multiply(A=Saturate(None=@n6), B=OneMinus(None=@n8))
   B=CollectionParameter() {collection=/Game/Environments/TechArt/Hologram/MPC_Hologram, parameterName=ScreensEnabled}
  )
  n10 = Subtract(A=ObjectPositionWS(), B=WorldPosition().XYZ)
  If(A=@n9, B=Constant() {r=0.01}, "A > B"=Constant3Vector(), "A == B"=@n10, "A < B"=@n10)
  MaterialFunctionCall(Distance=ScalarParameter() {parameterName=CameraFadeDistance}) {materialFunction=&Near_Camera_Fade}
  n11 = MaterialFunctionCall!(
   Sphere Radius=Constant() {r=1}
   Sphere Falloff=@n4
   Unit Ray Direction=Normalize(VectorInput=@n2)
   Sphere Center=ObjectPositionWS()
   Ray Origin=@n1
  ) {materialFunction=/Game/Athena/Environments/MaterialFunctions/RayTraceSphereFalloffMinClamp}
  n12 = DepthFade(
   Opacity=Power(Base=OneMinus(None=@n11.Clamped Falloff), Exp=ScalarParameter() {defaultValue=1.5, parameterName=FalloffFog})
   FadeDistance=ScalarParameter() {defaultValue=50, parameterName=DepthFade}
  )
  n13 = Constant3Vector() {constant=<rgba 0 0 -1 0>}
  n14 = Add(
   A=DotProduct(A=VertexNormalWS(), B=Transform(None=@n13) {transformSourceType=TRANSFORMSOURCE_Local})
   B=ScalarParameter() {group=Hemisphere_Falloff, parameterName=Hemisphere_Falloff_Position}
  )
  n15 = Max(A=ScalarParameter() {defaultValue=0.0216, group=Hemisphere_Falloff, parameterName=Hemisphere_Falloff_Distance}) {constB=0}
  n16 = Saturate(None=Power(Base=Divide(A=@n14, B=@n15), Exp=ScalarParameter() {defaultValue=2, group=Hemisphere_Falloff, parameterName=Hemisphere_Falloff_Power}))
  Multiply(A=StaticSwitchParameter(True=Multiply(A=@n12, B=@n16), False=@n12) {group=Hemisphere_Falloff, parameterName=UseHemisphereFalloff}, B=@n9)
  n17 = ScalarParameter() {defaultValue=1, parameterName=Brightness_night}
  n18 = VectorParameter() {parameterName=Color_A, defaultValue=<rgba 0 0.0818 0.25 1>}
  n19 = VectorParameter() {parameterName=Color, defaultValue=<rgba 0.2335 0.25 0 1>}
  n20 = CollectionParameter() {collection=&MPC_Hologram, parameterName=ColorSwitch}
  n21 = NamedRerouteDeclaration(None=CollectionParameter() {collection=&MPC_Hologram, parameterName=Captured}) {name=Captured, nodeColor=<rgba 0.9219 1 0 1>}
  n22 = NamedRerouteDeclaration(None=Saturate(None=Add(A=@n20, B=@n21))) {name=ColorSwitch, nodeColor=<rgba 0 0.2266 1 1>}
  Multiply(
   A=MaterialFunctionCall(Value_Morning=@n17, Value_Day=ScalarParameter() {defaultValue=0.5, parameterName=Brightness_Day}, Value_Evening=@n17, Value_Night=@n17) {materialFunction=&Scalar_Time_of_Day}.Result
   B=LinearInterpolate(
    A=LinearInterpolate(A=@n18.RGB, B=@n19.RGB, Alpha=@n22)
    B=VectorParameter() {parameterName=Color_Contested, defaultValue=<rgba 0.6719 0 0 1>}.RGB
    Alpha=CollectionParameter() {collection=&MPC_Hologram, parameterName=Contested}
   )
  )
  Transform(None=@n2) {transformSourceType=TRANSFORMSOURCE_World, transformType=TRANSFORM_Local}
 ]
}
